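; Creative pickup variant that shows a floating icon with text, colored and chosen by whether the local player is friendly, neutral or hostile to the owning team.
; A replicated Icon Info struct drives it: its OnRep and visibility updates rebuild the value bar widget and, on replication, subscribe to team or class changes.
; Clients show or hide the bar through a visible-to-team check; instances without cosmetic events destroy the component.
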(fn IsFriendlyTeam
  (params (TeamToCheck :type EFortTeam%20Enum :default "Spectator"))
  (return (!= (== (Teams|TeamtoIndex TeamToCheck) (Knobs|Team|GetTeamValueasIndex (Variables|Default|GetAllowed%20Team))) (Math|Conversions|ToBoolean(Integer) (Variables|Default|GetAffectsTeam)))))

(fn OnRep_IconInfo
  (ProgressWidget|UpdateFloatingBarWidget)
  (Default|ApplySubscriptiontoPlayerChangesTeamorClass true))

(fn SetPickupIconColor
  (bind (_icon_info_friendlytext_24_09faac8b43239578befcf79970d337a6 _icon_info_neutraltext_25_b2acd5fa4f0f6328cac3c8b551869d11 _icon_info_hostiletext_26_9ddb4e164b724a825ca547814b3f1127 _icon_info_friendlycolor_2_299876fb4dd07efcc13b8f97222edcf1 _icon_info_neutralcolor_4_ab52a95b4e95b6d576c24992c3b8f258 _icon_info_hostilecolor_6_e191372047aab319514576aaa54d3d3f) (Utilities|Struct|BreakCreativePickupIconInfo (Variables|GetIcon%20Info_FriendlyText_24_09FAAC8B43239578BEFCF79970D337A6)))
  (Class|CreativeValueBarComponent|SetCustomColor (Variables|Default|GetFloatingProgressBar) true (Utilities|Select :Index (Variables|Default|GetPlayer%20Affiliation) :Friendly _icon_info_friendlycolor_2_299876fb4dd07efcc13b8f97222edcf1 :Neutral _icon_info_neutralcolor_4_ab52a95b4e95b6d576c24992c3b8f258 :Hostile _icon_info_hostilecolor_6_e191372047aab319514576aaa54d3d3f) :!class /Game/Creative/Devices/Common/Components/Creative_ValueBar_Component.Creative_ValueBar_Component_C))

(fn SetPickupIconSettings
  (params (Icon%20Info :type Creative%20Pickup%20Icon%20Info%20Structure))
  (Variables|Default|SetIcon%20Info Icon%20Info))

(fn UpdateFloatingBarWidget
  (bind _affectsteam (Variables|Default|GetAffectsTeam))
  (bind _returnvalue (== _affectsteam 0))
  (bind _localhostileteams (LocalVariables|GetlocalHostileTeams))
  (bind _localfriendlyteam (LocalVariables|GetlocalFriendlyTeam))
  (bind _returnvalue_2 (select _returnvalue _localhostileteams _localfriendlyteam))
  (bind (_icon_info_friendlytext_24_09faac8b43239578befcf79970d337a6 _icon_info_neutraltext_25_b2acd5fa4f0f6328cac3c8b551869d11 _icon_info_hostiletext_26_9ddb4e164b724a825ca547814b3f1127 _icon_info_friendlycolor_2_299876fb4dd07efcc13b8f97222edcf1 _icon_info_neutralcolor_4_ab52a95b4e95b6d576c24992c3b8f258 _icon_info_hostilecolor_6_e191372047aab319514576aaa54d3d3f _icon_info_friendlyicon_10_c48c06e843fc1a24e35afdb54d47aff9 _icon_info_neutralicon_12_dc7afd9c43bfdb4c9cb68d90d2b0248b _icon_info_hostileicon_14_7a17f2804dbb0a922f08f0bc03d0c5c2) (Utilities|Struct|BreakCreativePickupIconInfo (Variables|GetIcon%20Info_FriendlyText_24_09FAAC8B43239578BEFCF79970D337A6)))
  (bind _floatingprogressbar (Variables|Default|GetFloatingProgressBar))
  (bind _allowed_team (Variables|Default|GetAllowed%20Team))
  (bind _floatingprogressbar_2 (Variables|Default|GetFloatingProgressBar))
  (bind (_icon_info_friendlytext_24_09faac8b43239578befcf79970d337a6_2 _icon_info_neutraltext_25_b2acd5fa4f0f6328cac3c8b551869d11_2 _icon_info_hostiletext_26_9ddb4e164b724a825ca547814b3f1127_2 _icon_info_friendlycolor_2_299876fb4dd07efcc13b8f97222edcf1_2 _icon_info_neutralcolor_4_ab52a95b4e95b6d576c24992c3b8f258_2 _icon_info_hostilecolor_6_e191372047aab319514576aaa54d3d3f_2 _icon_info_friendlyicon_10_c48c06e843fc1a24e35afdb54d47aff9_2 _icon_info_neutralicon_12_dc7afd9c43bfdb4c9cb68d90d2b0248b_2 _icon_info_hostileicon_14_7a17f2804dbb0a922f08f0bc03d0c5c2_2 _icon_info_visibletoteam_29_07fc326e4832b55248c35c822d3ba54d _icon_info_owningteam_36_de16a80f47bef561e41eb8bd33f275ee) (Utilities|Struct|BreakCreativePickupIconInfo (Variables|GetIcon%20Info_FriendlyText_24_09FAAC8B43239578BEFCF79970D337A6)))
  (Networking|CanExecuteCosmeticEvents
    (:True
      (bind _valid (CallFunction|VisibleToTeam))
      (if _valid
        (LocalVariables|SetlocalHostileTeams (Knobs|Team|GetTeamArray self))
        (Variables|Default|SetPlayer%20Affiliation (Team|GetTeamAffiliationBetweenTeams (+ (Team|GetActorPvPTeamIndex (Fortnite|GetFirstLocalFortPlayerController true)) 1) (+ (Math|Conversions|ToByte(Integer) _icon_info_owningteam_36_de16a80f47bef561e41eb8bd33f275ee) 1)))
        (Utilities|FlowControl|Sequence
          (:then_0
            ;[1 1 1] Parse Team Data
            (if (Knobs|Team|IsTeamNeutral (Variables|Default|GetAllowed%20Team))
              (Utilities|Array|Clear _returnvalue_2)
              (elif (Knobs|Team|IsTeamAny _allowed_team)
                (LocalVariables|SetlocalFriendlyTeam _returnvalue_2) ; Friendly teams take priority over hostile teams.
                (else
                  (Utilities|Array|Add (select (== (Variables|Default|GetAffectsTeam) 0) (LocalVariables|GetlocalFriendlyTeam) (LocalVariables|GetlocalHostileTeams)) (Knobs|Team|GetTeamValueasIndex _allowed_team))))))
          (:then_1
            (Default|BindEventtoOnWidgetInitialized _floatingprogressbar (EventDispatchers|CreateEvent :!event Set%20Pickup%20Icon%20Color))
            (Class|CreativeValueBarComponent|SpawnOrUpdateValueBar _floatingprogressbar (CallFunction|GetPickupStaticMeshComponent) self "/Game/Creative/Devices/Common/Components/Creative_FloatingValueBar.Creative_FloatingValueBar_C" 0 (Variables|Default|GetHealthBarDistance) 0 1.0 8e+01 false (LocalVariables|GetlocalFriendlyTeam) (LocalVariables|GetlocalHostileTeams) (Knobs|Team|GetShiftedAnyTeamValue) false (Utilities|String|ToString(Text) _icon_info_friendlytext_24_09faac8b43239578befcf79970d337a6) (Utilities|String|ToString(Text) _icon_info_neutraltext_25_b2acd5fa4f0f6328cac3c8b551869d11) (Utilities|String|ToString(Text) _icon_info_hostiletext_26_9ddb4e164b724a825ca547814b3f1127) 1e+01 false (Math|Boolean|ANDBoolean (Utilities|Text|TextIsEmpty _icon_info_friendlytext_24_09faac8b43239578befcf79970d337a6) (Utilities|Text|TextIsEmpty _icon_info_neutraltext_25_b2acd5fa4f0f6328cac3c8b551869d11) (Utilities|Text|TextIsEmpty _icon_info_hostiletext_26_9ddb4e164b724a825ca547814b3f1127)) true true (Utilities|Select :Index (Variables|Default|GetPlayer%20Affiliation) :Friendly _icon_info_friendlyicon_10_c48c06e843fc1a24e35afdb54d47aff9 :Neutral _icon_info_neutralicon_12_dc7afd9c43bfdb4c9cb68d90d2b0248b :Hostile _icon_info_hostileicon_14_7a17f2804dbb0a922f08f0bc03d0c5c2) true "(X=2.000000,Y=2.000000)" false false)))
        (else
          (Class|CreativeValueBarComponent|HideValueBar (Variables|Default|GetFloatingProgressBar)))))
    (:False
      (Utilities|IsValid _floatingprogressbar_2
        (:Is%20Valid
          (Components|DestroyComponent _floatingprogressbar_2))))))

(fn UpdateVisibilityAndCollision
  (|CallParentFunction)
  (ProgressWidget|UpdateFloatingBarWidget))

(fn ConstructionScript)

(fn VisibletoTeam
  (bind _icon_info (Variables|Default|GetIcon%20Info))
  (bind _visibletoteam_29_07fc326e4832b55248c35c822d3ba54d (Utilities|Struct|BreakCreativePickupIconInfo _icon_info))
  (bind _localpawnteamindex (LocalVariables|GetLocalPawnTeamIndex))
  (bind _friendlyteam (CallFunction|IsFriendlyTeam _localpawnteamindex))
  (LocalVariables|SetLocalPawnTeamIndex (Utilities|Enum|BytetoEnumEFortTeam (Team|GetActorPvPTeamIndex (Fortnite|GetFirstLocalFortPlayerController true))))
  (bind _index (CreativeTeam|GetTeamPVPIndex _visibletoteam_29_07fc326e4832b55248c35c822d3ba54d)
    (:Any%20/%20All
      (return true))
    (:Team%20Index
      (return (Utilities|Enum|Equal(Enum) (LocalVariables|GetLocalPawnTeamIndex) (Utilities|Enum|BytetoEnumEFortTeam _index))))
    (:Other
      (if (Creative|TeamOption|IsTeamType _visibletoteam_29_07fc326e4832b55248c35c822d3ba54d "Friendlies")
        (return _friendlyteam)
        (else
          (return (not _friendlyteam)))))))
